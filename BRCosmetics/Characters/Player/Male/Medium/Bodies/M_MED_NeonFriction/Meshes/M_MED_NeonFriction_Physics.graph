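# Physics asset for the M_MED_NeonFriction male body cosmetic, set up for secondary motion of its coat and sword.
# Coat capsule chains and the sword joint hang from master bones through limited swing and twist constraints.
# Angularly free constraints also tie the last 01 and 02 coat bodies together on each side.
# Pelvis, spine, leg and foot capsules supply body collision shapes, and a large offset sphere on root stands in for the floor.
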PhysicsAsset /BRCosmetics/Characters/Player/Male/Medium/Bodies/M_MED_NeonFriction/Meshes/M_MED_NeonFriction_Physics
bodies {
 calf_l [{shapeType=Capsule, length=35, radius=16, center=<-20 -2 0>, rotation=<rot 90 0 0>}]
 calf_r [{shapeType=Capsule, length=35, radius=16, center=<20 2 0>, rotation=<rot 90 0 0>}]
 dyn_coat_01_L_01 [{shapeType=Capsule, length=12, radius=6, shapeName=dyn_coat_01_L_01_capsule, center=<6 0 -10>, rotation=<rot 0 -0.0001 -30>}]
 dyn_coat_01_L_02 [{shapeType=Capsule, length=12, radius=8.5, shapeName=dyn_coat_01_L_02_capsule, center=<8.7 0 -14>, rotation=<rot 0 -0.0001 -30>}]
 dyn_coat_01_L_03 [{shapeType=Capsule, length=12, radius=9.5, shapeName=dyn_coat_01_L_03_capsule, center=<4.5 0 -14>, rotation=<rot 0 -0.0001 -30>}]
 dyn_coat_01_L_04 [{shapeType=Capsule, length=12, radius=12, shapeName=dyn_coat_01_L_04_capsule, center=<12 0 -14>, rotation=<rot 0 -0.0001 -30>}]
 dyn_coat_01_L_mstr [{shapeType=Capsule, length=0, radius=0, shapeName=dyn_coat_01_L_mstr_capsule}]
 dyn_coat_01_R_01 [{shapeType=Capsule, length=12, radius=6, shapeName=dyn_coat_01_L_01_capsule, center=<6 0 10>, rotation=<rot -0.0001 -0.0004 30>}]
 dyn_coat_01_R_02 [{shapeType=Capsule, length=12, radius=8.5, shapeName=dyn_coat_01_L_02_capsule, center=<8.7 0 14>, rotation=<rot -0.0001 -0.0004 30>}]
 dyn_coat_01_R_03 [{shapeType=Capsule, length=12, radius=9.5, shapeName=dyn_coat_01_L_03_capsule, center=<4.5 0 14>, rotation=<rot -0.0001 -0.0004 30>}]
 dyn_coat_01_R_04 [{shapeType=Capsule, length=12, radius=12, shapeName=dyn_coat_01_L_04_capsule, center=<12 0 14>, rotation=<rot -0.0001 -0.0004 30>}]
 dyn_coat_01_R_mstr [{shapeType=Capsule, length=0, radius=0, shapeName=dyn_coat_01_R_mstr_capsule}]
 dyn_coat_02_L_01 [{shapeType=Capsule, length=18, radius=5.5, shapeName=dyn_coat_02_L_01_capsule, center=<6 -4 0>}]
 dyn_coat_02_L_02 [{shapeType=Capsule, length=22, radius=8.5, shapeName=dyn_coat_02_L_02_capsule, center=<7.5 -4 0>}]
 dyn_coat_02_L_03 [{shapeType=Capsule, length=60, radius=12, shapeName=dyn_coat_02_L_03_capsule, center=<16 -8 0>}]
 dyn_coat_02_L_mstr [{shapeType=Capsule, length=0, radius=0, shapeName=dyn_coat_02_L_mstr_capsule}]
 dyn_coat_02_R_01 [{shapeType=Capsule, length=18, radius=5.5, shapeName=dyn_coat_02_L_01_capsule, center=<6 -4 0>}]
 dyn_coat_02_R_02 [{shapeType=Capsule, length=22, radius=8.5, shapeName=dyn_coat_02_L_01_capsule, center=<7.5 -4 0>}]
 dyn_coat_02_R_03 [{shapeType=Capsule, length=60, radius=12, shapeName=dyn_coat_02_L_03_capsule, center=<16 -8 0>}]
 dyn_coat_02_R_mstr [{shapeType=Capsule, length=0, radius=0, shapeName=dyn_coat_02_R_mstr_capsule}]
 dyn_coat_03_L_01 [{shapeType=Capsule, length=20.5, radius=7, shapeName=dyn_coat_03_L_01_capsule, center=<8.7694 -4 0>}]
 dyn_coat_03_L_02 [{shapeType=Capsule, length=25, radius=8, shapeName=dyn_coat_03_L_02_capsule, center=<6.2746 -4 0>}]
 dyn_coat_03_L_03 [{shapeType=Capsule, length=60, radius=12, shapeName=dyn_coat_03_L_03_capsule, center=<16 -12 0>}]
 dyn_coat_03_L_mstr [{shapeType=Capsule, length=0, radius=0, shapeName=dyn_coat_03_L_mstr_capsule}]
 dyn_coat_03_R_01 [{shapeType=Capsule, length=20.5, radius=7, shapeName=dyn_coat_03_R_01_capsule, center=<8.7694 -4 0>}]
 dyn_coat_03_R_02 [{shapeType=Capsule, length=25, radius=8, shapeName=dyn_coat_03_R_02_capsule, center=<6.2746 -4 0>}]
 dyn_coat_03_R_03 [{shapeType=Capsule, length=60, radius=12, shapeName=dyn_coat_03_R_03_capsule, center=<16 -12 0>}]
 dyn_coat_03_R_mstr [{shapeType=Capsule, length=0, radius=0, shapeName=dyn_coat_03_R_mstr_capsule}]
 dyn_coat_B_01 [{shapeType=Capsule, length=21, radius=7, shapeName=dyn_coat_B_01_capsule, center=<8.7477 -4 0>}]
 dyn_coat_B_02 [{shapeType=Capsule, length=25, radius=7, shapeName=dyn_coat_B_02_capsule, center=<6.2746 -4 0>}]
 dyn_coat_B_03 [{shapeType=Capsule, length=60, radius=12, shapeName=dyn_coat_B_03_capsule, center=<16 -8 0>}]
 dyn_coat_B_mstr [{shapeType=Capsule, length=0, radius=0, shapeName=dyn_coat_B_mstr_capsule}]
 dyn_sword_jnt [{shapeType=Capsule, length=100, radius=3, shapeName=dyn_sword_jnt_capsule, center=<-24 0 0>, rotation=<rot 0 88 88>}]
 dyn_sword_mstr [{shapeType=Capsule, length=0.505, radius=0.505, shapeName=dyn_sword_mstr_capsule, rotation=<rot 0 -90 0>}]
 foot_l [{shapeType=Capsule, length=12, radius=8, shapeName=foot_l_capsule, center=<-1.7 7 0>, rotation=<rot 0 0 90>}]
 foot_r [{shapeType=Capsule, length=12, radius=8, shapeName=foot_r_capsule, center=<1.7 -7 0>, rotation=<rot 0 0 90>}]
 pelvis [{shapeType=Capsule, length=6, radius=11, center=<0 -2 0>}]
 root [{shapeType=Capsule, length=0, radius=210, shapeName=floorCOL_sphere, center=<0 15 -193>, rotation=<rot 0 -90 0>}]
 spine_01 [{shapeType=Capsule, length=3, radius=12}]
 spine_02 [{shapeType=Capsule, length=6, radius=10, shapeName=spine_02_capsule}]
 spine_03 [{shapeType=Capsule, length=6, radius=11, shapeName=spine_03_capsule}]
 spine_04 [{shapeType=Capsule, length=6, radius=12.5, shapeName=spine_04_capsule}]
 thigh_l [{shapeType=Capsule, length=30, radius=10, center=<-20 -1 0>, rotation=<rot 0 85 95>}]
 thigh_r [{shapeType=Capsule, length=30, radius=10, shapeName=thigh_r_capsule, center=<20 1 0>, rotation=<rot 0 85 95>}]
}
constraints [
 {bone1Name=dyn_sword_jnt, bone2Name=dyn_sword_mstr, swing1LimitDegrees=5, swing2LimitDegrees=5, twistLimitDegrees=0}
 {bone1Name=dyn_coat_01_L_01, bone2Name=dyn_coat_01_L_mstr, swing1LimitDegrees=25, swing2LimitDegrees=20, twistLimitDegrees=2}
 {bone1Name=dyn_coat_01_L_02, bone2Name=dyn_coat_01_L_01, swing1LimitDegrees=25, swing2LimitDegrees=20, twistLimitDegrees=2}
 {bone1Name=dyn_coat_01_L_03, bone2Name=dyn_coat_01_L_02, swing1LimitDegrees=25, swing2LimitDegrees=20, twistLimitDegrees=20}
 {bone1Name=dyn_coat_01_L_04, bone2Name=dyn_coat_01_L_03, swing1LimitDegrees=25, swing2LimitDegrees=20, twistLimitDegrees=20}
 {bone1Name=dyn_coat_01_R_01, bone2Name=dyn_coat_01_R_mstr, swing1LimitDegrees=25, swing2LimitDegrees=20, twistLimitDegrees=2}
 {bone1Name=dyn_coat_01_R_02, bone2Name=dyn_coat_01_R_01, swing1LimitDegrees=25, swing2LimitDegrees=20, twistLimitDegrees=2}
 {bone1Name=dyn_coat_01_R_03, bone2Name=dyn_coat_01_R_02, swing1LimitDegrees=25, swing2LimitDegrees=20, twistLimitDegrees=20}
 {bone1Name=dyn_coat_01_R_04, bone2Name=dyn_coat_01_R_03, swing1LimitDegrees=25, swing2LimitDegrees=20, twistLimitDegrees=20}
 {bone1Name=dyn_coat_02_L_01, bone2Name=dyn_coat_02_L_mstr, swing1LimitDegrees=25, swing2LimitDegrees=20, twistLimitDegrees=2}
 {bone1Name=dyn_coat_02_L_02, bone2Name=dyn_coat_02_L_01, swing1LimitDegrees=25, swing2LimitDegrees=20, twistLimitDegrees=2}
 {bone1Name=dyn_coat_02_L_03, bone2Name=dyn_coat_02_L_02, swing1LimitDegrees=25, swing2LimitDegrees=20, twistLimitDegrees=2}
 {bone1Name=dyn_coat_02_R_01, bone2Name=dyn_coat_02_R_mstr, swing1LimitDegrees=25, swing2LimitDegrees=20, twistLimitDegrees=2}
 {bone1Name=dyn_coat_02_R_02, bone2Name=dyn_coat_02_R_01, swing1LimitDegrees=25, swing2LimitDegrees=20, twistLimitDegrees=2}
 {bone1Name=dyn_coat_02_R_03, bone2Name=dyn_coat_02_R_02, swing1LimitDegrees=25, swing2LimitDegrees=20, twistLimitDegrees=2}
 {bone1Name=dyn_coat_03_L_01, bone2Name=dyn_coat_03_L_mstr, swing1LimitDegrees=5, swing2LimitDegrees=5, twistLimitDegrees=2}
 {bone1Name=dyn_coat_03_L_02, bone2Name=dyn_coat_03_L_01, swing1LimitDegrees=25, swing2LimitDegrees=15, twistLimitDegrees=2}
 {bone1Name=dyn_coat_03_L_03, bone2Name=dyn_coat_03_L_02, swing1LimitDegrees=25, swing2LimitDegrees=15, twistLimitDegrees=2}
 {bone1Name=dyn_coat_03_R_01, bone2Name=dyn_coat_03_R_mstr, swing1LimitDegrees=5, swing2LimitDegrees=5, twistLimitDegrees=2}
 {bone1Name=dyn_coat_03_R_02, bone2Name=dyn_coat_03_R_01, swing1LimitDegrees=25, swing2LimitDegrees=15, twistLimitDegrees=2}
 {bone1Name=dyn_coat_03_R_03, bone2Name=dyn_coat_03_R_02, swing1LimitDegrees=25, swing2LimitDegrees=15, twistLimitDegrees=2}
 {bone1Name=dyn_coat_B_01, bone2Name=dyn_coat_B_mstr, swing1LimitDegrees=5, swing2LimitDegrees=5, twistLimitDegrees=2}
 {bone1Name=dyn_coat_B_02, bone2Name=dyn_coat_B_01, swing1LimitDegrees=25, swing2LimitDegrees=15, twistLimitDegrees=2}
 {bone1Name=dyn_coat_B_03, bone2Name=dyn_coat_B_02, swing1LimitDegrees=25, swing2LimitDegrees=15, twistLimitDegrees=2}
 {
  bone1Name=dyn_coat_02_L_03
  bone2Name=dyn_coat_01_L_04
  swing1LimitDegrees=45
  swing1Motion=Free
  swing2LimitDegrees=45
  swing2Motion=Free
  twistLimitDegrees=45
  twistMotion=Free
 }
 {
  bone1Name=dyn_coat_02_R_03
  bone2Name=dyn_coat_01_R_04
  swing1LimitDegrees=45
  swing1Motion=Free
  swing2LimitDegrees=45
  swing2Motion=Free
  twistLimitDegrees=45
  twistMotion=Free
 }
]
