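# Character-part AnimBP for the NeonFriction outfit, adding secondary motion to coat, sleeves and calf ties.
# Copied parent pose feeds a Control Rig, bone offsets, AnimDynamics and constraints into cached PreSim.
# Cloth-enabled uses PreSim directly; otherwise output blends to a pelvis-space RigidBody simulation with scaled coat master bones.
# EventGraph drives cloth, gravity and state alphas, and toggles material hide parameters for matching backpack and pickaxe.
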AnimBlueprint /BRCosmetics/Characters/Player/Male/Medium/Bodies/M_MED_NeonFriction/Meshes/M_MED_NeonFriction_AnimBP
target_skeleton=&/M_MED_NeonFriction_Skeleton
graphs [
 AnimGraph {nodes [
   Anim Dynamics - Bone: dyn_sword_jnt!(SimulationSpace:AnimPhysSimSpaceType Enum, ComponentPose, Alpha=1, ExternalForce="0,0,0") -> (Pose) {
    angularDampingOverride=0.9999
    angularSpringConstant=1200
    bAngularSpring=true
    bOverrideAngularDamping=true
    bUseGravityOverride=true
    simulationSpace=RootRelative
    boundBone {boneName=dyn_sword_jnt}
    componentAppliedLinearAccClamp=<3000 3000 3000>
    componentLinearAccScale=<0.4 0.4 0.4>
    componentLinearVelScale=<0.4 0.4 0.4>
    gravityOverride {z=-980}
    physicsBodyDefinitions [{
      boundBone {boneName=dyn_sword_jnt}
      boxExtents=<10 10 10>
      constraintSetup {angularLimitsMax=<1 5 5>, angularLimitsMin=<-1 -5 -5>, angularTarget=<1 0 0>}
      localJointOffset=<2 0 0>
     }]
   }
   n1 = "Constraint - dyn_coat_03_L_02 (PP)"!(ComponentPose, Alpha=1, ConstraintWeights_0=0.5, ConstraintWeights_1:float=0.5) -> (Pose) {
    alphaBoolBlend {blendOption=Linear}
    alphaScaleBiasClamp {scale=0.2}
    boneToModify {boneName=dyn_coat_03_L_02}
    constraintSetup [{targetBone {boneName=dyn_coat_02_L_02}}, {targetBone {boneName=dyn_coat_B_02}}]
    constraintWeights [1,1]
   }
   n2 = "Constraint - dyn_coat_03_L_03 (PP)"!(ComponentPose=@n1, Alpha=1, ConstraintWeights_0=0.5, ConstraintWeights_1:float=0.5) -> (Pose) {
    alphaBoolBlend {blendOption=Linear}
    alphaScaleBiasClamp {scale=0.3}
    boneToModify {boneName=dyn_coat_03_L_03}
    constraintSetup [{targetBone {boneName=dyn_coat_02_L_03}}, {targetBone {boneName=dyn_coat_B_03}}]
    constraintWeights [1,1]
   }
   n3 = "Constraint - dyn_coat_03_R_02 (PP)"!(ComponentPose=@n2, Alpha=1, ConstraintWeights_0=0.5, ConstraintWeights_1:float=0.5) -> (Pose) {
    alphaBoolBlend {blendOption=Linear}
    alphaScaleBiasClamp {scale=0.2}
    boneToModify {boneName=dyn_coat_03_R_02}
    constraintSetup [{targetBone {boneName=dyn_coat_02_R_02}}, {targetBone {boneName=dyn_coat_B_02}}]
    constraintWeights [1,1]
   }
   n4 = "Constraint - dyn_coat_03_R_03 (PP)"!(ComponentPose=@n3, Alpha=1, ConstraintWeights_0=0.5, ConstraintWeights_1:float=0.5) -> (Pose) {
    alphaBoolBlend {blendOption=Linear}
    alphaScaleBiasClamp {scale=0.3}
    boneToModify {boneName=dyn_coat_03_R_03}
    constraintSetup [{targetBone {boneName=dyn_coat_02_R_03}}, {targetBone {boneName=dyn_coat_B_03}}]
    constraintWeights [1,1]
   }
   n5 = "Constraint - dyn_coat_B_02 (PP)"!(ComponentPose=@n4, Alpha=1, ConstraintWeights_0=0.5, ConstraintWeights_1:float=0.5) -> (Pose) {
    alphaBoolBlend {blendOption=Linear}
    alphaScaleBiasClamp {scale=0.2}
    boneToModify {boneName=dyn_coat_B_02}
    constraintSetup [{targetBone {boneName=dyn_coat_03_L_02}}, {targetBone {boneName=dyn_coat_03_R_02}}]
    constraintWeights [1,1]
   }
   "Constraint - dyn_coat_B_03 (PP)"!(ComponentPose=@n5, Alpha=1, ConstraintWeights_0=0.5, ConstraintWeights_1:float=0.5) -> (Pose) {
    alphaBoolBlend {blendOption=Linear}
    alphaScaleBiasClamp {scale=0.3}
    boneToModify {boneName=dyn_coat_B_03}
    constraintSetup [{targetBone {boneName=dyn_coat_03_L_03}}, {targetBone {boneName=dyn_coat_03_R_03}}]
    constraintWeights [1,1]
   }
   Output Pose(Result=Use cached pose 'PostSim'()) {layerGroup=None}
   n6 = Control Rig!(
    Source=Copy Pose From Mesh() {bUseAttachedParent=true}
    Alpha
    Movement Alpha:double=Get MovementAlpha() -> (MovementAlpha:double)
    is_DBNO_Carried:bool=Get bIsDBNOCarried()
    IsSkydivingAlpha:double=Get Is Skydiving Alpha() -> (Is Skydiving Alpha:double)
   ) -> (Pose) {
    alphaBoolBlend {blendOption=Linear}
    controlRigAssetReference {blueprintRigClass=&/M_MED_NeonFriction_CtrlRig.M_MED_NeonFriction_CtrlRig_C}
   }
   n7 = "Transform (Modify) Bone - Bone: backpack_collision"(ComponentPose=Local To Component(LocalPose=@n6), Alpha=1, Translation="0,0,200", Rotation="0,0,0", Scale="1,1,1") {
    rotationMode=BMM_Additive
    rotationSpace=BCS_BoneSpace
    translationMode=BMM_Additive
    translationSpace=BCS_BoneSpace
    boneToModify {boneName=backpack_collision}
    translation {z=200}
   }
   n8 = "Transform (Modify) Bone - Bone: backbling_lowerstrap_l"(ComponentPose=@n7, Alpha=1, Translation="2,1,1", Rotation="0,0,0", Scale="1,1,1") {
    rotationMode=BMM_Additive
    rotationSpace=BCS_BoneSpace
    translationMode=BMM_Additive
    translationSpace=BCS_BoneSpace
    boneToModify {boneName=backbling_lowerstrap_l}
    translation=<2 1 1>
   }
   n9 = "Transform (Modify) Bone - Bone: backbling_lowerstrap_r"(ComponentPose=@n8, Alpha=1, Translation="2,-1,1", Rotation="0,0,0", Scale="1,1,1") {
    rotationMode=BMM_Additive
    rotationSpace=BCS_BoneSpace
    translationMode=BMM_Additive
    translationSpace=BCS_BoneSpace
    boneToModify {boneName=backbling_lowerstrap_r}
    translation=<2 -1 1>
   }
   n10 = "Transform (Modify) Bone - Bone: dyn_coat_01_R_mstr"(ComponentPose=@n9, Alpha=1, Translation="0,0,0", Rotation="20,-35,0", Scale="1,1,1") {
    rotationMode=BMM_Additive
    rotationSpace=BCS_BoneSpace
    scaleMode=BMM_Replace
    scaleSpace=BCS_BoneSpace
    translationMode=BMM_Additive
    translationSpace=BCS_BoneSpace
    alphaBoolBlend {blendOption=Linear}
    alphaScaleBiasClamp {interpSpeedDecreasing=4}
    boneToModify {boneName=dyn_coat_01_R_mstr}
    rotation {pitch=20, yaw=-35}
   }
   n11 = "Transform (Modify) Bone - Bone: dyn_coat_01_L_mstr"(ComponentPose=@n10, Alpha=1, Translation="0,0,0", Rotation="-20,-10,0", Scale="1,1,1") {
    rotationMode=BMM_Additive
    rotationSpace=BCS_BoneSpace
    scaleMode=BMM_Replace
    scaleSpace=BCS_BoneSpace
    translationMode=BMM_Additive
    translationSpace=BCS_BoneSpace
    alphaBoolBlend {blendOption=Linear}
    alphaScaleBiasClamp {interpSpeedDecreasing=4}
    boneToModify {boneName=dyn_coat_01_L_mstr}
    rotation {pitch=-20, yaw=-10}
   }
   n12 = "Transform (Modify) Bone - Bone: dyn_coat_03_R_mstr"(ComponentPose=@n11, Alpha=1, Translation="0,0,0", Rotation="0,-25,0", Scale="1,1,1") {
    rotationMode=BMM_Additive
    rotationSpace=BCS_BoneSpace
    scaleMode=BMM_Replace
    scaleSpace=BCS_BoneSpace
    translationMode=BMM_Additive
    translationSpace=BCS_BoneSpace
    alphaBoolBlend {blendOption=Linear}
    alphaScaleBiasClamp {interpSpeedDecreasing=4}
    boneToModify {boneName=dyn_coat_03_R_mstr}
    rotation {yaw=-25}
   }
   n13 = "Transform (Modify) Bone - Bone: dyn_coat_02_R_mstr"(ComponentPose=@n12, Alpha=1, Translation="0,0,0", Rotation="20,-35,0", Scale="1,1,1") {
    rotationMode=BMM_Additive
    rotationSpace=BCS_BoneSpace
    scaleMode=BMM_Replace
    scaleSpace=BCS_BoneSpace
    translationMode=BMM_Additive
    translationSpace=BCS_BoneSpace
    alphaBoolBlend {blendOption=Linear}
    alphaScaleBiasClamp {interpSpeedDecreasing=4}
    boneToModify {boneName=dyn_coat_02_R_mstr}
    rotation {pitch=20, yaw=-35}
   }
   n14 = "Transform (Modify) Bone - Bone: dyn_coat_03_L_mstr"(ComponentPose=@n13, Alpha=1, Translation="0,0,0", Rotation="0,-5,0", Scale="1,1,1") {
    rotationMode=BMM_Additive
    rotationSpace=BCS_BoneSpace
    scaleMode=BMM_Replace
    scaleSpace=BCS_BoneSpace
    translationMode=BMM_Additive
    translationSpace=BCS_BoneSpace
    alphaBoolBlend {blendOption=Linear}
    alphaScaleBiasClamp {interpSpeedDecreasing=4}
    boneToModify {boneName=dyn_coat_03_L_mstr}
    rotation {yaw=-5}
   }
   n15 = "Transform (Modify) Bone - Bone: dyn_coat_02_L_mstr"(ComponentPose=@n14, Alpha=1, Translation="0,0,0", Rotation="0,-5,0", Scale="1,1,1") {
    rotationMode=BMM_Additive
    rotationSpace=BCS_BoneSpace
    scaleMode=BMM_Replace
    scaleSpace=BCS_BoneSpace
    translationMode=BMM_Additive
    translationSpace=BCS_BoneSpace
    alphaBoolBlend {blendOption=Linear}
    alphaScaleBiasClamp {interpSpeedDecreasing=4}
    boneToModify {boneName=dyn_coat_02_L_mstr}
    rotation {yaw=-5}
   }
   n16 = "Transform (Modify) Bone - Bone: dyn_coat_B_mstr"(ComponentPose=@n15, Alpha=1, Translation="0,0,0", Rotation="0,-10,0", Scale="1,1,1") {
    rotationMode=BMM_Additive
    rotationSpace=BCS_BoneSpace
    scaleMode=BMM_Replace
    scaleSpace=BCS_BoneSpace
    translationMode=BMM_Additive
    translationSpace=BCS_BoneSpace
    alphaBoolBlend {blendOption=Linear}
    alphaScaleBiasClamp {interpSpeedDecreasing=4}
    boneToModify {boneName=dyn_coat_B_mstr}
    rotation {yaw=-10}
   }
   n17 = Anim Dynamics - Bone: dyn_l_calftie_mid_01!(SimulationSpace:AnimPhysSimSpaceType Enum, ComponentPose=@n16, Alpha=1, GravityOverride:Vector="0,0,-980", ExternalForce="0,0,0") -> (Pose) {
    angularDampingOverride=0.9999
    angularSpringConstant=50
    bAngularSpring=true
    bOverrideAngularDamping=true
    bUseGravityOverride=true
    gravityScale=0.35
    simulationSpace=RootRelative
    alphaBoolBlend {blendOption=Linear}
    boundBone {boneName=dyn_l_calftie_mid_01}
    componentLinearAccScale=<0.4 0.4 0.4>
    componentLinearVelScale=<0.4 0.4 0.4>
    gravityOverride {z=-980}
    physicsBodyDefinitions [{
      boundBone {boneName=dyn_l_calftie_mid_01}
      boxExtents=<14 14 14>
      constraintSetup {angularLimitsMax=<0 20 1>, angularLimitsMin=<0 -20 -1>, angularTarget=<1 0 0>}
      localJointOffset=<2 0 0>
     }]
   }
   n18 = Anim Dynamics - Bone: dyn_l_calftie_low_01!(SimulationSpace:AnimPhysSimSpaceType Enum, ComponentPose=@n17, Alpha=1, GravityOverride:Vector="0,0,-980", ExternalForce="0,0,0") -> (Pose) {
    angularDampingOverride=0.9999
    angularSpringConstant=50
    bAngularSpring=true
    bOverrideAngularDamping=true
    bUseGravityOverride=true
    gravityScale=0.35
    simulationSpace=RootRelative
    alphaBoolBlend {blendOption=Linear}
    boundBone {boneName=dyn_l_calftie_low_01}
    componentLinearAccScale=<0.4 0.4 0.4>
    componentLinearVelScale=<0.4 0.4 0.4>
    gravityOverride {z=-980}
    physicsBodyDefinitions [{
      boundBone {boneName=dyn_l_calftie_low_01}
      boxExtents=<14 14 14>
      constraintSetup {angularLimitsMax=<0 20 1>, angularLimitsMin=<0 -20 -1>, angularTarget=<1 0 0>}
      localJointOffset=<2 0 0>
     }]
   }
   n19 = Anim Dynamics - Bone: dyn_l_calftie_up_01!(SimulationSpace:AnimPhysSimSpaceType Enum, ComponentPose=@n18, Alpha=1, GravityOverride:Vector="0,0,-980", ExternalForce="0,0,0") -> (Pose) {
    angularDampingOverride=0.9999
    angularSpringConstant=50
    bAngularSpring=true
    bOverrideAngularDamping=true
    bUseGravityOverride=true
    gravityScale=0.35
    simulationSpace=RootRelative
    alphaBoolBlend {blendOption=Linear}
    boundBone {boneName=dyn_l_calftie_up_01}
    componentLinearAccScale=<0.4 0.4 0.4>
    componentLinearVelScale=<0.4 0.4 0.4>
    gravityOverride {z=-980}
    physicsBodyDefinitions [{
      boundBone {boneName=dyn_l_calftie_up_01}
      boxExtents=<14 14 14>
      constraintSetup {angularLimitsMax=<0 20 1>, angularLimitsMin=<0 -20 -1>, angularTarget=<1 0 0>}
      localJointOffset=<2 0 0>
     }]
   }
   n20 = Anim Dynamics - Bone: dyn_r_calftie_mid_01!(SimulationSpace:AnimPhysSimSpaceType Enum, ComponentPose=@n19, Alpha=1, GravityOverride:Vector="0,0,-980", ExternalForce="0,0,0") -> (Pose) {
    angularDampingOverride=0.9999
    angularSpringConstant=50
    bAngularSpring=true
    bOverrideAngularDamping=true
    bUseGravityOverride=true
    gravityScale=0.35
    simulationSpace=RootRelative
    alphaBoolBlend {blendOption=Linear}
    boundBone {boneName=dyn_r_calftie_mid_01}
    componentLinearAccScale=<0.4 0.4 0.4>
    componentLinearVelScale=<0.4 0.4 0.4>
    gravityOverride {z=-980}
    physicsBodyDefinitions [{
      boundBone {boneName=dyn_r_calftie_mid_01}
      boxExtents=<14 14 14>
      constraintSetup {angularLimitsMax=<0 20 1>, angularLimitsMin=<0 -20 -1>, angularTarget=<1 0 0>}
      localJointOffset=<-2 0 0>
     }]
   }
   n21 = Anim Dynamics - Bone: dyn_r_calftie_low_01!(SimulationSpace:AnimPhysSimSpaceType Enum, ComponentPose=@n20, Alpha=1, GravityOverride:Vector="0,0,-980", ExternalForce="0,0,0") -> (Pose) {
    angularDampingOverride=0.9999
    angularSpringConstant=50
    bAngularSpring=true
    bOverrideAngularDamping=true
    bUseGravityOverride=true
    gravityScale=0.35
    simulationSpace=RootRelative
    alphaBoolBlend {blendOption=Linear}
    boundBone {boneName=dyn_r_calftie_low_01}
    componentLinearAccScale=<0.4 0.4 0.4>
    componentLinearVelScale=<0.4 0.4 0.4>
    gravityOverride {z=-980}
    physicsBodyDefinitions [{
      boundBone {boneName=dyn_r_calftie_low_01}
      boxExtents=<14 14 14>
      constraintSetup {angularLimitsMax=<0 20 1>, angularLimitsMin=<0 -20 -1>, angularTarget=<1 0 0>}
      localJointOffset=<-2 0 0>
     }]
   }
   n22 = Anim Dynamics - Bone: dyn_r_calftie_up_01!(SimulationSpace:AnimPhysSimSpaceType Enum, ComponentPose=@n21, Alpha=1, GravityOverride:Vector="0,0,-980", ExternalForce="0,0,0") -> (Pose) {
    angularDampingOverride=0.9999
    angularSpringConstant=50
    bAngularSpring=true
    bOverrideAngularDamping=true
    bUseGravityOverride=true
    gravityScale=0.35
    simulationSpace=RootRelative
    alphaBoolBlend {blendOption=Linear}
    boundBone {boneName=dyn_r_calftie_up_01}
    componentLinearAccScale=<0.4 0.4 0.4>
    componentLinearVelScale=<0.4 0.4 0.4>
    gravityOverride {z=-980}
    physicsBodyDefinitions [{
      boundBone {boneName=dyn_r_calftie_up_01}
      boxExtents=<14 14 14>
      constraintSetup {angularLimitsMax=<0 20 1>, angularLimitsMin=<0 -20 -1>, angularTarget=<1 0 0>}
      localJointOffset=<-2 0 0>
     }]
   }
   n23 = Anim Dynamics - Bone: dyn_sleeve_l_tp_01!(SimulationSpace:AnimPhysSimSpaceType Enum, ComponentPose=@n22, Alpha=1, GravityOverride:Vector="0,0,-980", ExternalForce="0,0,0") -> (Pose) {
    angularDampingOverride=0.9999
    angularSpringConstant=50
    bAngularSpring=true
    bOverrideAngularDamping=true
    bUseGravityOverride=true
    gravityScale=0.35
    simulationSpace=RootRelative
    alphaBoolBlend {blendOption=Linear}
    boundBone {boneName=dyn_sleeve_l_tp_01}
    componentLinearAccScale=<0.4 0.4 0.4>
    componentLinearVelScale=<0.4 0.4 0.4>
    gravityOverride {z=-980}
    physicsBodyDefinitions [{
      boundBone {boneName=dyn_sleeve_l_tp_01}
      boxExtents=<14 14 14>
      constraintSetup {angularLimitsMax=<1 5 3>, angularLimitsMin=<-1 -5 -3>, angularTarget=<1 0 0>}
      localJointOffset=<2 0 0>
     }]
   }
   n24 = Anim Dynamics - Bone: dyn_sleeve_l_tp_02!(SimulationSpace:AnimPhysSimSpaceType Enum, ComponentPose=@n23, Alpha=1, GravityOverride:Vector="0,0,-980", ExternalForce="0,0,0") -> (Pose) {
    angularDampingOverride=0.9999
    angularSpringConstant=50
    bAngularSpring=true
    bOverrideAngularDamping=true
    bUseGravityOverride=true
    gravityScale=0.35
    simulationSpace=RootRelative
    alphaBoolBlend {blendOption=Linear}
    boundBone {boneName=dyn_sleeve_l_tp_02}
    componentLinearAccScale=<0.4 0.4 0.4>
    componentLinearVelScale=<0.4 0.4 0.4>
    gravityOverride {z=-980}
    physicsBodyDefinitions [{
      boundBone {boneName=dyn_sleeve_l_tp_02}
      boxExtents=<14 14 14>
      constraintSetup {angularLimitsMax=<1 5 3>, angularLimitsMin=<-1 -5 -3>, angularTarget=<1 0 0>}
      localJointOffset=<2 0 0>
     }]
   }
   n25 = Anim Dynamics - Bone: dyn_sleeve_l_bt_01!(SimulationSpace:AnimPhysSimSpaceType Enum, ComponentPose=@n24, Alpha=1, GravityOverride:Vector="0,0,-980", ExternalForce="0,0,0") -> (Pose) {
    angularDampingOverride=0.9999
    angularSpringConstant=50
    bAngularSpring=true
    bOverrideAngularDamping=true
    bUseGravityOverride=true
    gravityScale=0.35
    simulationSpace=RootRelative
    alphaBoolBlend {blendOption=Linear}
    boundBone {boneName=dyn_sleeve_l_bt_01}
    componentLinearAccScale=<0.4 0.4 0.4>
    componentLinearVelScale=<0.4 0.4 0.4>
    gravityOverride {z=-980}
    physicsBodyDefinitions [{
      boundBone {boneName=dyn_sleeve_l_bt_01}
      boxExtents=<14 14 14>
      constraintSetup {angularLimitsMax=<1 5 3>, angularLimitsMin=<-1 -5 -3>, angularTarget=<1 0 0>}
      localJointOffset=<2 0 0>
     }]
   }
   n26 = Anim Dynamics - Bone: dyn_sleeve_l_bt_02!(SimulationSpace:AnimPhysSimSpaceType Enum, ComponentPose=@n25, Alpha=1, GravityOverride:Vector="0,0,-980", ExternalForce="0,0,0") -> (Pose) {
    angularDampingOverride=0.9999
    angularSpringConstant=50
    bAngularSpring=true
    bOverrideAngularDamping=true
    bUseGravityOverride=true
    gravityScale=0.35
    simulationSpace=RootRelative
    alphaBoolBlend {blendOption=Linear}
    boundBone {boneName=dyn_sleeve_l_bt_02}
    componentLinearAccScale=<0.4 0.4 0.4>
    componentLinearVelScale=<0.4 0.4 0.4>
    gravityOverride {z=-980}
    physicsBodyDefinitions [{
      boundBone {boneName=dyn_sleeve_l_bt_02}
      boxExtents=<14 14 14>
      constraintSetup {angularLimitsMax=<1 5 3>, angularLimitsMin=<-1 -5 -3>, angularTarget=<1 0 0>}
      localJointOffset=<2 0 0>
     }]
   }
   n27 = Anim Dynamics - Bone: dyn_sleeve_l_bk_01!(SimulationSpace:AnimPhysSimSpaceType Enum, ComponentPose=@n26, Alpha=1, GravityOverride:Vector="0,0,-980", ExternalForce="0,0,0") -> (Pose) {
    angularDampingOverride=0.9999
    angularSpringConstant=50
    bAngularSpring=true
    bOverrideAngularDamping=true
    bUseGravityOverride=true
    gravityScale=0.35
    simulationSpace=RootRelative
    alphaBoolBlend {blendOption=Linear}
    boundBone {boneName=dyn_sleeve_l_bk_01}
    componentLinearAccScale=<0.4 0.4 0.4>
    componentLinearVelScale=<0.4 0.4 0.4>
    gravityOverride {z=-980}
    physicsBodyDefinitions [{
      boundBone {boneName=dyn_sleeve_l_bk_01}
      boxExtents=<14 14 14>
      constraintSetup {angularLimitsMax=<1 5 3>, angularLimitsMin=<-1 -5 -3>, angularTarget=<1 0 0>}
      localJointOffset=<2 0 0>
     }]
   }
   n28 = Anim Dynamics - Bone: dyn_sleeve_l_bk_02!(SimulationSpace:AnimPhysSimSpaceType Enum, ComponentPose=@n27, Alpha=1, GravityOverride:Vector="0,0,-980", ExternalForce="0,0,0") -> (Pose) {
    angularDampingOverride=0.9999
    angularSpringConstant=50
    bAngularSpring=true
    bOverrideAngularDamping=true
    bUseGravityOverride=true
    gravityScale=0.35
    simulationSpace=RootRelative
    alphaBoolBlend {blendOption=Linear}
    boundBone {boneName=dyn_sleeve_l_bk_02}
    componentLinearAccScale=<0.4 0.4 0.4>
    componentLinearVelScale=<0.4 0.4 0.4>
    gravityOverride {z=-980}
    physicsBodyDefinitions [{
      boundBone {boneName=dyn_sleeve_l_bk_02}
      boxExtents=<14 14 14>
      constraintSetup {angularLimitsMax=<1 5 3>, angularLimitsMin=<-1 -5 -3>, angularTarget=<1 0 0>}
      localJointOffset=<2 0 0>
     }]
   }
   n29 = Anim Dynamics - Bone: dyn_sleeve_l_ft_01!(SimulationSpace:AnimPhysSimSpaceType Enum, ComponentPose=@n28, Alpha=1, GravityOverride:Vector="0,0,-980", ExternalForce="0,0,0") -> (Pose) {
    angularDampingOverride=0.9999
    angularSpringConstant=50
    bAngularSpring=true
    bOverrideAngularDamping=true
    bUseGravityOverride=true
    gravityScale=0.35
    simulationSpace=RootRelative
    alphaBoolBlend {blendOption=Linear}
    boundBone {boneName=dyn_sleeve_l_ft_01}
    componentLinearAccScale=<0.4 0.4 0.4>
    componentLinearVelScale=<0.4 0.4 0.4>
    gravityOverride {z=-980}
    physicsBodyDefinitions [{
      boundBone {boneName=dyn_sleeve_l_ft_01}
      boxExtents=<14 14 14>
      constraintSetup {angularLimitsMax=<1 5 3>, angularLimitsMin=<-1 -5 -3>, angularTarget=<1 0 0>}
      localJointOffset=<2 0 0>
     }]
   }
   n30 = Anim Dynamics - Bone: dyn_sleeve_l_ft_02!(SimulationSpace:AnimPhysSimSpaceType Enum, ComponentPose=@n29, Alpha=1, GravityOverride:Vector="0,0,-980", ExternalForce="0,0,0") -> (Pose) {
    angularDampingOverride=0.9999
    angularSpringConstant=50
    bAngularSpring=true
    bOverrideAngularDamping=true
    bUseGravityOverride=true
    gravityScale=0.35
    simulationSpace=RootRelative
    alphaBoolBlend {blendOption=Linear}
    boundBone {boneName=dyn_sleeve_l_ft_02}
    componentLinearAccScale=<0.4 0.4 0.4>
    componentLinearVelScale=<0.4 0.4 0.4>
    gravityOverride {z=-980}
    physicsBodyDefinitions [{
      boundBone {boneName=dyn_sleeve_l_ft_02}
      boxExtents=<14 14 14>
      constraintSetup {angularLimitsMax=<1 5 3>, angularLimitsMin=<-1 -5 -3>, angularTarget=<1 0 0>}
      localJointOffset=<2 0 0>
     }]
   }
   n31 = Anim Dynamics - Bone: dyn_sleeve_r_tp_01!(SimulationSpace:AnimPhysSimSpaceType Enum, ComponentPose=@n30, Alpha=1, GravityOverride:Vector="0,0,-980", ExternalForce="0,0,0") -> (Pose) {
    angularDampingOverride=0.9999
    angularSpringConstant=50
    bAngularSpring=true
    bOverrideAngularDamping=true
    bUseGravityOverride=true
    gravityScale=0.35
    simulationSpace=RootRelative
    alphaBoolBlend {blendOption=Linear}
    boundBone {boneName=dyn_sleeve_r_tp_01}
    componentLinearAccScale=<0.4 0.4 0.4>
    componentLinearVelScale=<0.4 0.4 0.4>
    gravityOverride {z=-980}
    physicsBodyDefinitions [{
      boundBone {boneName=dyn_sleeve_r_tp_01}
      boxExtents=<14 14 14>
      constraintSetup {angularLimitsMax=<1 5 3>, angularLimitsMin=<-1 -5 -3>, angularTarget=<1 0 0>}
      localJointOffset=<-2 0 0>
     }]
   }
   n32 = Anim Dynamics - Bone: dyn_sleeve_r_tp_02!(SimulationSpace:AnimPhysSimSpaceType Enum, ComponentPose=@n31, Alpha=1, GravityOverride:Vector="0,0,-980", ExternalForce="0,0,0") -> (Pose) {
    angularDampingOverride=0.9999
    angularSpringConstant=50
    bAngularSpring=true
    bOverrideAngularDamping=true
    bUseGravityOverride=true
    gravityScale=0.35
    simulationSpace=RootRelative
    alphaBoolBlend {blendOption=Linear}
    boundBone {boneName=dyn_sleeve_r_tp_02}
    componentLinearAccScale=<0.4 0.4 0.4>
    componentLinearVelScale=<0.4 0.4 0.4>
    gravityOverride {z=-980}
    physicsBodyDefinitions [{
      boundBone {boneName=dyn_sleeve_r_tp_02}
      boxExtents=<14 14 14>
      constraintSetup {angularLimitsMax=<1 5 3>, angularLimitsMin=<-1 -5 -3>, angularTarget=<1 0 0>}
      localJointOffset=<-2 0 0>
     }]
   }
   n33 = Anim Dynamics - Bone: dyn_sleeve_r_bt_01!(SimulationSpace:AnimPhysSimSpaceType Enum, ComponentPose=@n32, Alpha=1, GravityOverride:Vector="0,0,-980", ExternalForce="0,0,0") -> (Pose) {
    angularDampingOverride=0.9999
    angularSpringConstant=50
    bAngularSpring=true
    bOverrideAngularDamping=true
    bUseGravityOverride=true
    gravityScale=0.35
    simulationSpace=RootRelative
    alphaBoolBlend {blendOption=Linear}
    boundBone {boneName=dyn_sleeve_r_bt_01}
    componentLinearAccScale=<0.4 0.4 0.4>
    componentLinearVelScale=<0.4 0.4 0.4>
    gravityOverride {z=-980}
    physicsBodyDefinitions [{
      boundBone {boneName=dyn_sleeve_r_bt_01}
      boxExtents=<14 14 14>
      constraintSetup {angularLimitsMax=<1 5 3>, angularLimitsMin=<-1 -5 -3>, angularTarget=<1 0 0>}
      localJointOffset=<-2 0 0>
     }]
   }
   n34 = Anim Dynamics - Bone: dyn_sleeve_r_bt_02!(SimulationSpace:AnimPhysSimSpaceType Enum, ComponentPose=@n33, Alpha=1, GravityOverride:Vector="0,0,-980", ExternalForce="0,0,0") -> (Pose) {
    angularDampingOverride=0.9999
    angularSpringConstant=50
    bAngularSpring=true
    bOverrideAngularDamping=true
    bUseGravityOverride=true
    gravityScale=0.35
    simulationSpace=RootRelative
    alphaBoolBlend {blendOption=Linear}
    boundBone {boneName=dyn_sleeve_r_bt_02}
    componentLinearAccScale=<0.4 0.4 0.4>
    componentLinearVelScale=<0.4 0.4 0.4>
    gravityOverride {z=-980}
    physicsBodyDefinitions [{
      boundBone {boneName=dyn_sleeve_r_bt_02}
      boxExtents=<14 14 14>
      constraintSetup {angularLimitsMax=<1 5 3>, angularLimitsMin=<-1 -5 -3>, angularTarget=<1 0 0>}
      localJointOffset=<-2 0 0>
     }]
   }
   n35 = Anim Dynamics - Bone: dyn_sleeve_r_bk_01!(SimulationSpace:AnimPhysSimSpaceType Enum, ComponentPose=@n34, Alpha=1, GravityOverride:Vector="0,0,-980", ExternalForce="0,0,0") -> (Pose) {
    angularDampingOverride=0.9999
    angularSpringConstant=50
    bAngularSpring=true
    bOverrideAngularDamping=true
    bUseGravityOverride=true
    gravityScale=0.35
    simulationSpace=RootRelative
    alphaBoolBlend {blendOption=Linear}
    boundBone {boneName=dyn_sleeve_r_bk_01}
    componentLinearAccScale=<0.4 0.4 0.4>
    componentLinearVelScale=<0.4 0.4 0.4>
    gravityOverride {z=-980}
    physicsBodyDefinitions [{
      boundBone {boneName=dyn_sleeve_r_bk_01}
      boxExtents=<14 14 14>
      constraintSetup {angularLimitsMax=<1 5 3>, angularLimitsMin=<-1 -5 -3>, angularTarget=<1 0 0>}
      localJointOffset=<-2 0 0>
     }]
   }
   n36 = Anim Dynamics - Bone: dyn_sleeve_r_bk_02!(SimulationSpace:AnimPhysSimSpaceType Enum, ComponentPose=@n35, Alpha=1, GravityOverride:Vector="0,0,-980", ExternalForce="0,0,0") -> (Pose) {
    angularDampingOverride=0.9999
    angularSpringConstant=50
    bAngularSpring=true
    bOverrideAngularDamping=true
    bUseGravityOverride=true
    gravityScale=0.35
    simulationSpace=RootRelative
    alphaBoolBlend {blendOption=Linear}
    boundBone {boneName=dyn_sleeve_r_bk_02}
    componentLinearAccScale=<0.4 0.4 0.4>
    componentLinearVelScale=<0.4 0.4 0.4>
    gravityOverride {z=-980}
    physicsBodyDefinitions [{
      boundBone {boneName=dyn_sleeve_r_bk_02}
      boxExtents=<14 14 14>
      constraintSetup {angularLimitsMax=<1 5 3>, angularLimitsMin=<-1 -5 -3>, angularTarget=<1 0 0>}
      localJointOffset=<-2 0 0>
     }]
   }
   n37 = Anim Dynamics - Bone: dyn_sleeve_r_ft_01!(SimulationSpace:AnimPhysSimSpaceType Enum, ComponentPose=@n36, Alpha=1, GravityOverride:Vector="0,0,-980", ExternalForce="0,0,0") -> (Pose) {
    angularDampingOverride=0.9999
    angularSpringConstant=50
    bAngularSpring=true
    bOverrideAngularDamping=true
    bUseGravityOverride=true
    gravityScale=0.35
    simulationSpace=RootRelative
    alphaBoolBlend {blendOption=Linear}
    boundBone {boneName=dyn_sleeve_r_ft_01}
    componentLinearAccScale=<0.4 0.4 0.4>
    componentLinearVelScale=<0.4 0.4 0.4>
    gravityOverride {z=-980}
    physicsBodyDefinitions [{
      boundBone {boneName=dyn_sleeve_r_ft_01}
      boxExtents=<14 14 14>
      constraintSetup {angularLimitsMax=<1 5 3>, angularLimitsMin=<-1 -5 -3>, angularTarget=<1 0 0>}
      localJointOffset=<-2 0 0>
     }]
   }
   n38 = Anim Dynamics - Bone: dyn_sleeve_r_ft_02!(SimulationSpace:AnimPhysSimSpaceType Enum, ComponentPose=@n37, Alpha=1, GravityOverride:Vector="0,0,-980", ExternalForce="0,0,0") -> (Pose) {
    angularDampingOverride=0.9999
    angularSpringConstant=50
    bAngularSpring=true
    bOverrideAngularDamping=true
    bUseGravityOverride=true
    gravityScale=0.35
    simulationSpace=RootRelative
    alphaBoolBlend {blendOption=Linear}
    boundBone {boneName=dyn_sleeve_r_ft_02}
    componentLinearAccScale=<0.4 0.4 0.4>
    componentLinearVelScale=<0.4 0.4 0.4>
    gravityOverride {z=-980}
    physicsBodyDefinitions [{
      boundBone {boneName=dyn_sleeve_r_ft_02}
      boxExtents=<14 14 14>
      constraintSetup {angularLimitsMax=<1 5 3>, angularLimitsMin=<-1 -5 -3>, angularTarget=<1 0 0>}
      localJointOffset=<-2 0 0>
     }]
   }
   n39 = "Constraint - dyn_coat_01_L_02 (P)"(ComponentPose=@n38, Alpha=1, ConstraintWeights_0=1) {boneToModify {boneName=dyn_coat_01_L_02}, constraintSetup [{targetBone {boneName=dyn_coat_02_L_01}}], constraintWeights [1]}
   n40 = "Constraint - dyn_coat_01_R_02 (P)"(ComponentPose=@n39, Alpha=1, ConstraintWeights_0=1) {boneToModify {boneName=dyn_coat_01_R_02}, constraintSetup [{targetBone {boneName=dyn_coat_02_R_01}}], constraintWeights [1]}
   n41 = "Transform (Modify) Bone - Bone: dyn_coat_03_R_02"(ComponentPose=@n40, Alpha=1, Translation="0,-5,0", Rotation="0,0,0", Scale="1,1,1") {
    rotationMode=BMM_Additive
    rotationSpace=BCS_BoneSpace
    scaleMode=BMM_Replace
    scaleSpace=BCS_BoneSpace
    translationMode=BMM_Additive
    translationSpace=BCS_BoneSpace
    alphaBoolBlend {blendOption=Linear}
    alphaScaleBiasClamp {interpSpeedDecreasing=4}
    boneToModify {boneName=dyn_coat_03_R_02}
    translation {y=-5}
   }
   n42 = "Transform (Modify) Bone - Bone: dyn_coat_03_L_02"(ComponentPose=@n41, Alpha=1, Translation="0,-5,0", Rotation="0,0,0", Scale="1,1,1") {
    rotationMode=BMM_Additive
    rotationSpace=BCS_BoneSpace
    scaleMode=BMM_Replace
    scaleSpace=BCS_BoneSpace
    translationMode=BMM_Additive
    translationSpace=BCS_BoneSpace
    alphaBoolBlend {blendOption=Linear}
    alphaScaleBiasClamp {interpSpeedDecreasing=4}
    boneToModify {boneName=dyn_coat_03_L_02}
    translation {y=-5}
   }
   n43 = "Transform (Modify) Bone - Bone: dyn_coat_B_02"(ComponentPose=@n42, Alpha=1, Translation="0,-5,0", Rotation="0,0,0", Scale="1,1,1") {
    rotationMode=BMM_Additive
    rotationSpace=BCS_BoneSpace
    scaleMode=BMM_Replace
    scaleSpace=BCS_BoneSpace
    translationMode=BMM_Additive
    translationSpace=BCS_BoneSpace
    alphaBoolBlend {blendOption=Linear}
    alphaScaleBiasClamp {interpSpeedDecreasing=4}
    boneToModify {boneName=dyn_coat_B_02}
    translation {y=-5}
   }
   n44 = "Transform (Modify) Bone - Bone: attach_backpack"(ComponentPose=@n43, Alpha=1, Translation="0,-5,0", Rotation="0,0,5", Scale="1,1,1") {
    rotationMode=BMM_Additive
    rotationSpace=BCS_BoneSpace
    scaleMode=BMM_Replace
    scaleSpace=BCS_BoneSpace
    translationMode=BMM_Additive
    translationSpace=BCS_BoneSpace
    alphaBoolBlend {blendOption=Linear}
    alphaScaleBiasClamp {interpSpeedDecreasing=4}
    boneToModify {boneName=attach_backpack}
    rotation {roll=5}
    translation {y=-5}
   }
   Save cached pose 'PreSim'(Pose=Component To Local(ComponentPose=@n44)) {cacheName=PreSim}
   n45 = RigidBody!(ComponentPose=Local To Component(LocalPose=Use cached pose 'PreSim'()), Alpha, OverrideWorldGravity:Vector="0,0,0", ExternalForce) -> (Pose) {
    bOverrideWorldGravity=true
    overridePhysicsAsset=&/M_MED_NeonFriction_Physics
    simulationSpace=BaseBoneSpace
    alphaBoolBlend {blendOption=Linear}
    baseBoneRef {boneName=pelvis}
    componentAppliedLinearAccClamp=<3000 3000 3000>
    simSpaceSettings {
     dampingAlpha=0
     maxAngularAcceleration=40
     maxAngularVelocity=8
     maxLinearAcceleration=3500
     maxLinearVelocity=700
     velocityScaleZ=0.5
     worldAlpha=1
     externalLinearDragV=<3 3 3>
    }
   }
   n46 = "Transform (Modify) Bone - Bone: dyn_coat_03_R_mstr"(ComponentPose=@n45, Alpha=1, Translation="0,0,0", Rotation="0,0,0", Scale="0.4,1,1") {
    rotationMode=BMM_Additive
    rotationSpace=BCS_BoneSpace
    scaleMode=BMM_Replace
    scaleSpace=BCS_BoneSpace
    translationMode=BMM_Additive
    translationSpace=BCS_BoneSpace
    alphaBoolBlend {blendOption=Linear}
    alphaScaleBiasClamp {interpSpeedDecreasing=4}
    boneToModify {boneName=dyn_coat_03_R_mstr}
    scale {x=0.4}
   }
   n47 = "Transform (Modify) Bone - Bone: dyn_coat_02_R_mstr"(ComponentPose=@n46, Alpha=1, Translation="0,0,0", Rotation="0,0,0", Scale="0.4,1,1") {
    rotationMode=BMM_Additive
    rotationSpace=BCS_BoneSpace
    scaleMode=BMM_Replace
    scaleSpace=BCS_BoneSpace
    translationMode=BMM_Additive
    translationSpace=BCS_BoneSpace
    alphaBoolBlend {blendOption=Linear}
    alphaScaleBiasClamp {interpSpeedDecreasing=4}
    boneToModify {boneName=dyn_coat_02_R_mstr}
    scale {x=0.4}
   }
   n48 = "Transform (Modify) Bone - Bone: dyn_coat_03_L_mstr"(ComponentPose=@n47, Alpha=1, Translation="0,0,0", Rotation="0,0,0", Scale="0.4,1,1") {
    rotationMode=BMM_Additive
    rotationSpace=BCS_BoneSpace
    scaleMode=BMM_Replace
    scaleSpace=BCS_BoneSpace
    translationMode=BMM_Additive
    translationSpace=BCS_BoneSpace
    alphaBoolBlend {blendOption=Linear}
    alphaScaleBiasClamp {interpSpeedDecreasing=4}
    boneToModify {boneName=dyn_coat_03_L_mstr}
    scale {x=0.4}
   }
   n49 = "Transform (Modify) Bone - Bone: dyn_coat_02_L_mstr"(ComponentPose=@n48, Alpha=1, Translation="0,0,0", Rotation="0,0,0", Scale="0.4,1,1") {
    rotationMode=BMM_Additive
    rotationSpace=BCS_BoneSpace
    scaleMode=BMM_Replace
    scaleSpace=BCS_BoneSpace
    translationMode=BMM_Additive
    translationSpace=BCS_BoneSpace
    alphaBoolBlend {blendOption=Linear}
    alphaScaleBiasClamp {interpSpeedDecreasing=4}
    boneToModify {boneName=dyn_coat_02_L_mstr}
    scale {x=0.4}
   }
   n50 = "Transform (Modify) Bone - Bone: dyn_coat_B_mstr"(ComponentPose=@n49, Alpha=1, Translation="0,0,0", Rotation="0,0,0", Scale="0.4,1,1") {
    rotationMode=BMM_Additive
    rotationSpace=BCS_BoneSpace
    scaleMode=BMM_Replace
    scaleSpace=BCS_BoneSpace
    translationMode=BMM_Additive
    translationSpace=BCS_BoneSpace
    alphaBoolBlend {blendOption=Linear}
    alphaScaleBiasClamp {interpSpeedDecreasing=4}
    boneToModify {boneName=dyn_coat_B_mstr}
    scale {x=0.4}
   }
   n51 = "Transform (Modify) Bone - Bone: dyn_coat_01_L_mstr"(ComponentPose=@n50, Alpha=1, Translation="0,0,0", Rotation="0,0,0", Scale="0.4,1,1") {
    rotationMode=BMM_Additive
    rotationSpace=BCS_BoneSpace
    scaleMode=BMM_Replace
    scaleSpace=BCS_BoneSpace
    translationMode=BMM_Additive
    translationSpace=BCS_BoneSpace
    alphaBoolBlend {blendOption=Linear}
    alphaScaleBiasClamp {interpSpeedDecreasing=4}
    boneToModify {boneName=dyn_coat_01_L_mstr}
    scale {x=0.4}
   }
   n52 = "Transform (Modify) Bone - Bone: dyn_coat_01_R_mstr"(ComponentPose=@n51, Alpha=1, Translation="0,0,0", Rotation="0,0,0", Scale="0.4,1,1") {
    rotationMode=BMM_Additive
    rotationSpace=BCS_BoneSpace
    scaleMode=BMM_Replace
    scaleSpace=BCS_BoneSpace
    translationMode=BMM_Additive
    translationSpace=BCS_BoneSpace
    alphaBoolBlend {blendOption=Linear}
    alphaScaleBiasClamp {interpSpeedDecreasing=4}
    boneToModify {boneName=dyn_coat_01_R_mstr}
    scale {x=0.4}
   }
   Save cached pose 'PostSim'(Pose=Blend Poses by bool(
     BlendPose_0=Use cached pose 'PreSim'()
     BlendPose_1=Component To Local(ComponentPose=@n52)
     BlendTime_0=0.2
     BlendTime_1=0.5
     bActiveValue=Get bClothEnabled()
    ) {blendType=Linear}) {cacheName=PostSim}
  ]}
 EventGraph {nodes [
   n1 = Event BlueprintUpdateAnimation()
   > n2 = Sequence!(execute) -> (then_0, then_1, then_2:Exec, then_3:Exec, then_4:Exec, then_5:Exec)
   > Set simSpace(simSpace:AnimPhysSimSpaceType Enum) -> (Output_Get:AnimPhysSimSpaceType Enum)
   > Branch(Condition=OR Boolean(A=Get bIsDBNO(), B=Get bIsInFrontEnd()))
   > n3 = Branch(Condition=OR Boolean(A=Get bIsDBNOCarried(), B=Get bIsInFrontEnd()))
   > Set simSpace(simSpace=World)
   Set simSpace(execute=@n3.else)
   Set DeltaTimeLocal(execute=@n2.then_2, DeltaTimeLocal:double=@n1.DeltaTimeX) -> (Output_Get:double)
   > n4 = Branch(Condition=Get bClothEnabled())
   > n5 = Sequence!(execute) -> (then_0, then_1)
   n6 = "Gravity Override State Function (Parameter Asset) (Nativized)"(
    execute=[@n5.then_0, @n4.else]
    CharacterPartAnimInstance=Self-Reference()
    GameTimeInSeconds=GetGameTimeInSeconds()
    Parameters=&/M_MED_NeonFriction_Parameters
   )
   > Set GravityOverride(GravityOverride=@n6)
   n7 = "Cloth Controller Function (Parameter Asset) (Nativized)"!(
    execute=@n5.then_1
    CharacterPartAnimInstance:Custom Character Part Anim Instance Object Reference=Self-Reference()
    ClothingAssetName:Name=M_MED_NeonFriction_Coat_Clothing
    ClothCoreSettings:"Cloth Core Settings Structure (by ref)"=Get Cloth Core Settings Native() -> (Cloth Core Settings Native:Cloth Core Settings Structure)
    WindGustLocalSettings:"Wind Gust Settings Structure (by ref)"=Get Wind Gust Local Settings Native() -> (Wind Gust Local Settings Native:Wind Gust Settings Structure)
    ClothUpAxis:Vector="1,0,0"
    LinVelScaleFwdBwdClamp:float=0
    LinVelScaleUpDwnClamp:float=0.7
    DeltaTime:float=Get DeltaTimeLocal() -> (DeltaTimeLocal:double)
    GameTimeInSeconds:float=GetGameTimeInSeconds()
    StateLogicLayerName:Name=none
    Parameters:Fort Character Dynamics Parameters Object Reference=&M_MED_NeonFriction_Parameters
   ) -> (
    then
    OutCurrentState:Name=None
    OutClothParameters:Cloth Parameters Structure
    OutClothCoreSettings:Cloth Core Settings Structure
    OutWindGustSettings:Wind Gust Parameters Structure
    OutWindGustLocalSettings:Wind Gust Settings Structure
    OutWindVelocity:Vector
    OutLinearVelVector:Vector
   )
   > Set Wind Gust Local Settings Native Out(Wind Gust Local Settings Native Out:Wind Gust Settings Structure=@n7.OutWindGustLocalSettings) -> (Output_Get:Wind Gust Settings Structure)
   n8 = PlayerMovement(execute=@n2.then_3, AnimInstance=Self-Reference(), XDotProductCoefficient=25, YDotProductCoefficient=25, ZDotProductCoefficient=25)
   > Set MovementAlpha(MovementAlpha:double="Absolute (Float)"(A=float * float(A=@n8.YMovement, B=0.045) {is=title K2Node_PromotableOperator})) -> (Output_Get:double)
   n9 = Sequence!(execute=@n2.then_4) -> (then_0, then_1, then_2:Exec, then_3:Exec, then_4:Exec, then_5:Exec)
   > Set Is Baseball Sliding(Is Baseball Sliding:double="To Float (Boolean)"(InBool=Get bIsBaseballSliding())) -> (Output_Get:double)
   Set Is In Vehicle Alpha(execute=@n9.then_1, Is In Vehicle Alpha:double="To Float (Boolean)"(InBool=Get bIsInVehicle() -> (bIsInVehicle:bool))) -> (Output_Get:double)
   Set Is in FrontEnd Float(execute=@n9.then_2, Is in FrontEnd Float:double="To Float (Boolean)"(InBool=Get bIsInFrontEnd())) -> (Output_Get:double)
   Set Is On GrindRail Alpha(execute=@n9.then_3, Is On GrindRail Alpha:double="To Float (Boolean)"(InBool=Get bIsOnGrindRail())) -> (Output_Get:double)
   Set Is Skydiving Alpha(execute=@n9.then_4, Is Skydiving Alpha:double="To Float (Boolean)"(InBool=Get bIsSkyDiving())) -> (Output_Get:double)
   > Set Is Parachute Opened Alpha(Is Parachute Opened Alpha:double="To Float (Boolean)"(InBool=Get bIsParachuteOpened())) -> (Output_Get:double)
   Set frontendAlpha(execute=@n2.then_1, frontendAlpha:double=SelectFloat(A=0.5, B=1, bPickA=Get bIsInFrontEnd())) -> (Output_Get:double)
   n10 = Event OnCharacterCustomizationLoaded()
   n11 = Sequence!(execute=[@n10.then, @n20.then, @n23.then, @n24.then]) -> (then_0, then_1, then_2)
   n12 = CheckForBackbling()
   n13 = OR Boolean(A=@n12, B=@n12.bIsBackblingLong)
   > Set Is Wearing Backbling(Is Wearing Backbling=@n13)
   > Set Is Not Wearing Backbling(Is Not Wearing Backbling:bool=NOT Boolean(A=@n13)) -> (Output_Get:bool)
   n14 = Get Dynamic Material Instance(execute=@n11.then_1, SkeletalMeshComponent=GetPartSkeletalMeshComponent())
   > n15 = Branch(Condition=BackpackItemDefinitionHasAnyMetaTag!(Tags:Array of Gameplay Tag Structures=Make Array("[0]"=(TagName="Cosmetics.MetaTags.NeonFriction")), bRequiresEquipped:bool=true) -> (ReturnValue:bool))
   > SetScalarParameterValue(self=@n14, ParameterName=Hide Element 01, Value=0)
   SetScalarParameterValue(execute=@n15.else, self=@n14, ParameterName=Hide Element 01, Value=1)
   n16 = PickaxeItemDefinitionHasAnyMetaTag(Tags=Make Array("[0]"=(TagName="Cosmetics.MetaTags.NeonFriction")), bRequiresEquipped=true)
   n17 = Branch(execute=@n11.then_2, Condition=@n16)
   > n18 = Branch(Condition=OR Boolean!(A=Get bIsSkyDiving(), B=Get bIsTacticalSprinting(), C:bool=Get bIsPlayingEmote()) -> (ReturnValue))
   > n19 = Branch(Condition=Get bIsCrouchSprinting())
   SetScalarParameterValue(execute=[@n22.then, @n19.then, @n18.else], self=@n14, ParameterName=Hide Element 02, Value=1)
   SetScalarParameterValue(execute=[@n21.then, @n22.else, @n19.else, @n17.else], self=@n14, ParameterName=Hide Element 02, Value=0)
   n20 = Event OnWeaponChanged()
   n21 = Event OnBeginSkydiving()
   Event OnEndSkydiving()
   > n22 = Branch(Condition=@n16)
   n23 = Event OnParachuteOpened()
   n24 = Event OnParachuteClosed()
   DrawVectorFromJointFunction!(
    execute
    SkelMeshComponent:Scene Component Object Reference
    JointName:Name=pelvis
    LineEnd:Vector=Break Cloth Setting Struct!(ClothSettingStruct:"Cloth Setting Struct Structure (by ref)") -> (
     LinearVelocityScale_7_#1:Vector=~
     AngularVelocityScale_6_#2:float=~
     FictitiousAngularScale_10_#3:float=~
     GravityOverride_15_#4:Vector=~
     DragCoefficient_23_#5:Vector 2D Structure
     LiftCoefficient_24_#6:Vector 2D Structure
     JointName_27_#7:Name
    ) {is=title K2Node_BreakStruct}.GravityOverride_15_#4
    JointOffset:Vector="0, 0, 0"
    LineColor:Linear Color Structure=(R=1,G=0,B=0,A=1)
    Thickness:double=2
    Foreground:EDrawDebugSceneDepthPriorityGroup Enum=Foreground
   ) -> (then)
  ]}
]
variables [
 Cloth Core Settings Native {
  type=ClothCoreSettings
  default {
   collisionThickness=0
   dampingCoefficient=0.05
   frictionCoefficient=0.3
   localDampingCoefficient=0
   numberIterations=2
   numberSubsteps=1
   selfCollisionThickness=0
   animDriveDamping=<0 1>
   animDriveStiffness=<0 1>
   areaStiffness=<1 1>
   bendingStiffness=<1 1>
   edgeStiffness=<1 1>
   tetherStiffness=<1 1>
  }
 }
 DeltaTimeLocal {type=number}
 frontendAlpha {type=number}
 Gravity Override Params {type=GravityOverrideParamsStruct}
 GravityOverride {type=Vector}
 Is Baseball Sliding {type=number}
 Is in FrontEnd Float {type=number}
 Is In Vehicle Alpha {type=number}
 Is Not Wearing Backbling {type=boolean}
 Is On GrindRail Alpha {type=number}
 Is Parachute Opened Alpha {type=number}
 Is Skydiving Alpha {type=number}
 Is Wearing Backbling {type=boolean}
 MovementAlpha {type=number}
 simSpace {type=AnimPhysSimSpaceType}
 Wind Gust Local Settings Native {type=WindGustSettings, default {elapsedTime=0, finalGustStrengthLocal=0, gustDuration=0, gustStrength=0, timeBetweenGusts=0}}
 Wind Gust Local Settings Native Out {type=WindGustSettings, default {elapsedTime=0, finalGustStrengthLocal=0, gustDuration=0, gustStrength=0, timeBetweenGusts=0}}
]
