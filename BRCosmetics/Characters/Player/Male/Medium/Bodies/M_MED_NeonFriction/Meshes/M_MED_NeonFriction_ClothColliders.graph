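# Physics asset of cloth collision shapes for the NeonFriction medium male player character body.
# Capsules wrap the pelvis, spine, clavicles and neck, with a sphere on the head, while arm, leg and foot bodies are left empty.
# Extra shapes cover a backpack bone, the dynamic sword joints and a large ground box.
# It has no constraints, so it defines only collision volumes, with no ragdoll joints.
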PhysicsAsset /BRCosmetics/Characters/Player/Male/Medium/Bodies/M_MED_NeonFriction/Meshes/M_MED_NeonFriction_ClothColliders
bodies {
 backpack_collision [{shapeType=Capsule, length=24, radius=16, shapeName=backpack_collision_capsule, center=<0 -15.9239 -6.1253>, rotation=<rot 0 90 90>}]
 calf_l []
 calf_r []
 clavicle_l [{shapeType=Capsule, length=8, radius=7, center=<5 0.5 0>, rotation=<rot 79.3028 -1.4097 174.1847>}]
 clavicle_r [{shapeType=Capsule, length=8, radius=7, center=<-5 0.5 0>, rotation=<rot 79.3028 -1.4093 -5.8148>}]
 dyn_sword_jnt [
  {shapeType=Capsule, length=70, radius=5, shapeName=dyn_sword_jnt_capsule, center=<-38.548 1.4928 -0.5533>, rotation=<rot 0 87.9159 -87>}
  {shapeType=Capsule, length=30, radius=3, shapeName=dyn_sword_jnt_capsule, center=<11.1641 -0.3162 -0.604>, rotation=<rot -0.0001 87.9159 -92>}
 ]
 dyn_sword_mstr [{shapeType=Capsule, length=0.505, radius=0.505, shapeName=dyn_sword_mstr_capsule, rotation=<rot 0 -90 0>}]
 foot_l []
 foot_r []
 ground_collision [{shapeType=Box, extentX=250, extentY=250, extentZ=20, shapeName=ground_collision_box, center=<0 0 -10>, rotation=<rot 0 -90 0>}]
 head [{shapeType=Sphere, radius=17.3125, center=<11.028 -0.174 0>}]
 lowerarm_l []
 lowerarm_r []
 neck_02 [{shapeType=Capsule, length=13.4375, radius=6, center=<-0.2642 -0.3408 0>, rotation=<rot 90 -171.1193 -175.3978>}]
 pelvis [{shapeType=Capsule, length=6, radius=15, center=<-3 -0.6373 0>}]
 spine_01 [{shapeType=Capsule, length=8.5, radius=12, center=<-1.7965 -0.1477 0>}]
 spine_02 [
  {shapeType=Capsule, length=8, radius=11, center=<-1.711 0.8335 0.0001>}
  {shapeType=Capsule, length=29.3956, radius=10, center=<2.5024 0.4992 3.9825>, rotation=<rot 90 0.0049 0.0049>}
  {shapeType=Capsule, length=29.875, radius=10, center=<2.4704 0.4991 -3.9825>, rotation=<rot 90 -0.0396 -0.0396>}
 ]
 spine_03 [{shapeType=Capsule, length=8.2107, radius=11, center=<0.1445 0.5459 0.0001>}]
 spine_04 [{shapeType=Capsule, length=8, radius=12, center=<2.766 0.1447 0.0002>}]
 spine_05 [
  {shapeType=Capsule, length=8, radius=10, center=<-2.2795 -0.0739 0>}
  {shapeType=Capsule, length=13.9518, radius=10, center=<-0.2916 -1.7377 0>}
 ]
 thigh_l []
 thigh_r []
 upperarm_l []
 upperarm_r []
}
constraints []
